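# Editor utility widget for adjusting cinematic toon outline thickness on a character item definition's parts.
# Python extracts meshes and overlay materials; Generate duplicates a blank texture per part, or reuses an existing one, and wires it into those material instances.
# Load CID spawns paired preview actors on a rotator at a camera trace.
# A selection-change handler redirects protected actor picks to their selectable twins; Clear destroys preview actors.
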EditorUtilityWidgetBlueprint /Game/ArtTools/Tech_Art/TexturePainter/ToonLineAdjuster/EUW_ToonLineAdjuster
graphs [
 CheckForAlreadyExistingAsset {nodes [
   n1 = CheckForAlreadyExistingAsset!() -> (then, TargetName:String, TargetAssetLocationRef:String) {is=title K2Node_FunctionEntry}
   n2 = Split(SourceString=@n1.TargetAssetLocationRef, InStr=/, SearchDir=FromEnd)
   > n3 = GetAssetByObjectPath!(
    execute
    self:Asset Registry Interface=GetAssetRegistry()
    ObjectPath:Name=String To Name(InString=Append!(A=@n2, B=/, C:String=@n1, D:String=., E:String=@n1) -> (ReturnValue))
    bIncludeOnlyOnDiskAssets:bool=false
   ) -> (then, ReturnValue:Asset Data Structure)
   > Return Node!(
    execute
    AssetExists?:bool=Is Valid Asset Data!(InAssetData:"Asset Data Structure (by ref)"=@n3) -> (ReturnValue:bool)
    LoadedAsset:Object Reference=GetAsset(InAssetData=@n3)
   )
  ]}
 Cleanup {nodes [
   n1 = Get GeneratedTextures() -> (GeneratedTextures:Array of Texture Object References)
   n2 = DeleteLoadedAssets!(execute, AssetsToDelete:Array of Object References=@n1) -> (then, ReturnValue:bool)
   n3 = Get Selectables() -> (Selectables:Array of Skeletal Mesh Actor Object References)
   Branch(
    execute=[@n2.then, @n6.then]
    Condition=Is Array Non Empty!(InputArray:Array of Skeletal Mesh Actor Object References=@n3) -> (Non-Empty:bool=~) {is=title K2Node_MacroInstance}
   )
   n4 = Get Unselectables() -> (Unselectables:Array of Skeletal Mesh Actor Object References)
   > n5 = For Each Loop(Array:Array of Skeletal Mesh Actor Object References=@n4) -> (Array Element:Skeletal Mesh Actor Object Reference)
   Clear(execute=@n5.Completed, TargetArray:Array of Skeletal Mesh Actor Object References=@n3)
   > Clear(TargetArray=@n4)
   > Clear(TargetArray:Array of Texture Object References=@n1)
   > Clear(TargetArray:Array of Material Instance Dynamic Object References=Get MIDs() -> (MIDs:Array of Material Instance Dynamic Object References))
   > Clear(TargetArray:Array of Skeletal Mesh Object References=Get SkeletalMeshParts() -> (SkeletalMeshParts:Array of Skeletal Mesh Object References))
   > Destroy Actor(self=Get Rotator() -> (Rotator:BP Toonline Adjuster Rotator Object Reference))
   Destroy Actor(
    execute=@n5.LoopBody
    self=[
     "Get (a copy)"(Array:Array of Skeletal Mesh Actor Object References=@n3, Dimension 1=@n5.Array Index) -> (Output:Skeletal Mesh Actor Object Reference)
     @n5
    ]
   )
   n6 = Cleanup() {is=title K2Node_FunctionEntry}
   GetSelectedFolderPaths!() -> (ReturnValue:Array of Strings)
   n7 = GetSystemPath!(Object:Object Reference=Get LargeEmpty_Base() -> (LargeEmpty_Base:Texture Object Reference)) -> (ReturnValue:String)
   RenameAsset!(
    execute
    Asset:Object Reference=DuplicateAsset!(execute, SourceAssetPath:String=@n7, DestinationAssetPath:String=Append!(A=@n7, B=2221) -> (ReturnValue)) -> (then, ReturnValue:Object Reference)
    NewName:String
   ) -> (then)
   Get EditorAssetSubsystem() -> (ReturnValue:Editor Asset Subsystem Object Reference) {is=title K2Node_GetEditorSubsystem}
  ]}
 EUB_ToonLineAdjuster_AutoGenFunc {nodes [
   n1 = EUB_ToonLineAdjuster_AutoGenFunc!() -> (then, SelectionSet:Typed Element Selection Set Object Reference) {is=title K2Node_FunctionEntry}
   > Branch(Condition=Get ProtectOverlayOutlinerSelection?() -> (ProtectOverlayOutlinerSelection?:bool))
   > n2 = GetSelectedObjects!(execute, self:Typed Element Selection Set Object Reference=@n1, InRequiredClass:Object Class Reference=/Script/Engine.SkeletalMeshActor) -> (then, ReturnValue:Array of Object References)
   > n3 = For Each Loop(Array:Array of Object References=@n2) -> (Array Element:Object Reference)
   n4 = Cast To SkeletalMeshActor!(Object=@n3) -> (AsSkeletal Mesh Actor:Skeletal Mesh Actor Object Reference, bSuccess:bool=~) {is=title K2Node_DynamicCast}
   n5 = Get Unselectables()
   > Branch(Condition=Contains Item(TargetArray:Array of Skeletal Mesh Actor Object References=@n5, ItemToFind:"Skeletal Mesh Actor Object Reference (by ref)"=@n4))
   n6 = Get EditorActorSubsystem()
   > SelectNothing(self:Editor Actor Subsystem Object Reference=@n6)
   n7 = Find Item!(TargetArray:Array of Skeletal Mesh Actor Object References=@n5, ItemToFind:"Skeletal Mesh Actor Object Reference (by ref)"=@n4) -> (ReturnValue)
   > SetActorSelectionState!(
    execute
    self:Editor Actor Subsystem Object Reference=@n6
    Actor:Actor Object Reference="Get (a copy)"(Array=Get Selectables(), Dimension 1=@n7)
    bShouldBeSelected:bool=true
   ) -> (then)
   PrintString(InString="To String (Integer)"(InInt=@n7))
  ]}
 EventGraph {nodes [
   n1 = GetLevelViewportCameraInfo!(WorldContextObject:Object Reference) -> (CameraLocation:Vector, CameraRotation:Rotator, ReturnValue:bool)
   n2 = GetForwardVector!(InRot:Rotator=@n1.CameraRotation) -> (ReturnValue:Vector)
   n3 = Line Trace By Channel!(
    execute=[@n12.then, @n11.then]
    WorldContextObject:Object Reference
    Start:Vector=@n1
    End:Vector=vector + vector(A=@n1, B=vector * vector(A=@n2, B:float=1000) -> (ReturnValue=~))
    TraceChannel:ETraceTypeQuery Enum
    bTraceComplex:bool=false
    ActorsToIgnore:Array of Actor Object References
    DrawDebugType:EDrawDebugTrace Enum
    bIgnoreSelf:bool
    TraceColor:Linear Color Structure
    TraceHitColor:Linear Color Structure
    DrawTime:float
   ) -> (
    then
    OutHit_bBlockingHit:bool
    OutHit_bInitialOverlap:bool
    OutHit_Time:float
    OutHit_Distance:float
    OutHit_Location:Vector
    OutHit_ImpactPoint:Vector
    OutHit_Normal:Vector
    OutHit_ImpactNormal:Vector
    OutHit_PhysMat:Physical Material Object Reference
    OutHit_HitActor:Actor Object Reference
    OutHit_HitComponent:Primitive Component Object Reference
    OutHit_HitBoneName:Name=None
    OutHit_BoneName:Name=None
    OutHit_HitItem:int=0
    OutHit_ElementIndex:int=0
    OutHit_FaceIndex:int=0
    OutHit_TraceStart:Vector
    OutHit_TraceEnd:Vector
    ReturnValue:bool
   )
   n4 = SpawnActor Skeletal Mesh Actor!(
    execute:Exec=@n45.then_1
    WorldContextObject:Object Reference
    Class:Actor Class Reference=&SkeletalMeshActor
    SpawnTransform_Location:Vector=@n3.OutHit_Location
    SpawnTransform_Rotation:Rotator="0, 0, 0"
    SpawnTransform_Scale:Vector="1,1,1"
    CollisionHandlingOverride:ESpawnActorCollisionHandlingMethod Enum=Undefined
    TransformScaleMethod:ESpawnActorScaleMethod Enum=MultiplyWithRoot
    Owner:Actor Object Reference
    Instigator:Pawn Object Reference
   ) -> (then:Exec, ReturnValue:Skeletal Mesh Actor Object Reference) {is=title K2Node_SpawnActorFromClass}
   n5 = Get SkeletalMeshComponent!(self:Skeletal Mesh Actor Object Reference=@n4) -> (SkeletalMeshComponent:Skeletal Mesh Component Object Reference)
   n6 = For Each Loop(Exec=@n47.then, Array:Array of Skeletal Mesh Object References=Get SkeletalMeshParts()) -> (Array Element:Skeletal Mesh Object Reference)
   n7 = "Get (a copy)"(
    Array:Array of Material Interface Object References=Get OverlayMaterials() -> (OverlayMaterials:Array of Material Interface Object References)
    Dimension 1=@n6.Array Index
   ) -> (Output:Material Interface Object Reference)
   n8 = CreateDynamicMaterialInstance!(
    execute
    self:Primitive Component Object Reference=@n5
    ElementIndex:int=0
    SourceMaterial:Material Interface Object Reference=@n7
    OptionalName:Name=None
   ) -> (then, ReturnValue:Material Instance Dynamic Object Reference) {is=title K2Node_CallFunction}
   n9 = SetTextureParameterValue(
    self=@n8
    ParameterName=CinematicsOutlineThicknessOverride
    Value="Get (a copy)"(Array:Array of Texture Object References=Get GeneratedTextures(), Dimension 1=@n6.Array Index) -> (Output:Texture Object Reference)
   )
   n10 = For Each Loop(
    Exec=[@n9.then, @n8.then, @n43.then_1]
    Array:Array of Names=GetMaterialSlotNames!(self:Primitive Component Object Reference=@n5) -> (ReturnValue:Array of Names)
   ) -> (Array Element:Name)
   > SetMaterial(self=@n5, ElementIndex=@n10.Array Index, Material=@n7)
   "On Clicked (Clear)"!() -> (OutputDelegate, then, widget:User Widget Object Reference, event:Pointer Event Structure) {is=title K2Node_ComponentBoundEvent}
   > Cleanup() {is=title K2Node_CallFunction}
   "On Clicked (LoadCID)"!() -> (OutputDelegate, then, widget:User Widget Object Reference, event:Pointer Event Structure) {is=title K2Node_ComponentBoundEvent}
   > n11 = Branch(Condition=Is Array Non Empty(InputArray:Array of Skeletal Mesh Object References=Get SkeletalMeshParts()) -> (Non-Empty=~))
   n12 = Trigger Loading(execute=@n11.else)
   "On Clicked (GenerateTextures)"!() -> (OutputDelegate, then, widget:User Widget Object Reference, event:Pointer Event Structure) {is=title K2Node_ComponentBoundEvent}
   > Cleanup()
   > n13 = Sequence!(execute) -> (then_0, then_1, then_2:Exec)
   n14 = Execute Python Script!(
    execute=[@n36.then, @n13.then_0]
    PythonScript="import unreal\r\n\r\nskel_mesh = []\r\noverlays = []\r\n\r\nParts = id.get_editor_property(\"BaseCharacterParts\")\r\n##print(Parts)\r\n\r\nfor part in Parts:\r\n    skel_mesh.append(part.get_editor_property(\"SkeletalMesh\"))\r\n\r\n    OverlayData = part.get_editor_property(\"OverlayMaterialData\")\r\n    overlays.append(OverlayData.get_editor_property(\"HighQualityMaterial\"))\r\n\r\n##print(SkelMesh)\r\n##print(overlays)"
    ID:Athena Character Item Definition Object Reference=Get ID() -> (ID:Athena Character Item Definition Object Reference)
   ) -> (then, ReturnValue, SkelMesh:Array of Skeletal Mesh Object References, Overlays:Array of Material Interface Object References) {inputs [ID], outputs [SkelMesh,Overlays]}
   > n15 = Set SkeletalMeshParts(SkeletalMeshParts:Array of Skeletal Mesh Object References=@n14.SkelMesh) -> (Output_Get:Array of Skeletal Mesh Object References)
   > n16 = Set OverlayMaterials(OverlayMaterials:Array of Material Interface Object References=@n14.Overlays) -> (Output_Get:Array of Material Interface Object References)
   > n17 = For Each Loop(Array:Array of Skeletal Mesh Object References=@n15) -> (Array Element:Skeletal Mesh Object Reference)
   n18 = "Get (a copy)"(Array:Array of Material Interface Object References=@n16, Dimension 1=@n17.Array Index) -> (Output:Material Interface Object Reference)
   > PrintString(InString=Append!(A=GetDisplayName(Object=@n17), B=", ", C:String=GetDisplayName(Object=@n18)) -> (ReturnValue))
   Create Textures(execute=@n13.then_1)
   n19 = For Each Loop(Exec=@n13.then_2, Array:Array of Material Interface Object References=Get OverlayMaterials()) -> (Array Element:Material Interface Object Reference)
   > n20 = Cast To MaterialInstanceConstant(Object=@n19)
   > SetMaterialInstanceParent!(
    execute
    Instance:Material Instance Constant Object Reference=@n20
    NewParent:Material Interface Object Reference=Get Parent!(self:Material Instance Object Reference=@n20) -> (Parent:Material Interface Object Reference)
   ) -> (then)
   > SetMaterialInstanceStaticSwitchParameterValue!(
    execute
    Instance:Material Instance Constant Object Reference=@n20
    ParameterName:Name=UseCinematicsOutlineThicknessOverrideTexture
    Value:bool=true
    Association:EMaterialParameterAssociation Enum=GlobalParameter
    bUpdateMaterialInstance:bool=true
   ) -> (then, ReturnValue:bool)
   > SetMaterialInstanceTextureParameterValue!(
    execute
    Instance:Material Instance Constant Object Reference=@n20
    ParameterName:Name=CinematicsOutlineThicknessOverride
    Value:Texture Object Reference="Get (a copy)"(Array:Array of Texture Object References=Get GeneratedTextures(), Dimension 1=@n19.Array Index) -> (Output:Texture Object Reference)
    Association:EMaterialParameterAssociation Enum=GlobalParameter
   ) -> (then, ReturnValue:bool)
   Create Textures() {is=title K2Node_CustomEvent}
   > n21 = For Each Loop(Array:Array of Skeletal Mesh Object References=Get SkeletalMeshParts()) -> (Array Element:Skeletal Mesh Object Reference)
   n22 = Get EditorAssetSubsystem()
   > n23 = GetPathNameForLoadedAsset!(
    execute
    self:Editor Asset Subsystem Object Reference=@n22
    LoadedAsset="Get (a copy)"(Array:Array of Material Interface Object References=Get OverlayMaterials(), Dimension 1=@n21.Array Index) -> (Output:Material Interface Object Reference)
   ) -> (then, ReturnValue)
   n24 = Get LargeEmpty_Base()
   > n25 = GetPathNameForLoadedAsset(self=@n22, LoadedAsset=@n24)
   n26 = Split(SourceString=@n23, InStr=., SearchDir=FromEnd)
   n27 = GetDisplayName(Object=@n21)
   n28 = Append!(A=@n26, B=@n27, C=_CosmeticOverride, D:String=.uasset) -> (ReturnValue)
   n29 = Append!(A=@n27, B=_CosmeticOverride) -> (ReturnValue)
   > n30 = CheckForAlreadyExistingAsset!(execute, self, TargetName:String=@n29, TargetAssetLocationRef:String=@n28) -> (then, AssetExists?:bool, LoadedAsset:Object Reference) {is=title K2Node_CallFunction}
   > n31 = Branch(Condition=@n30)
   > PrintString(InString=PRevious asset detected)
   > n32 = Cast To Texture2D(Object=@n30.LoadedAsset)
   n33 = Get GeneratedTextures()
   > Add(TargetArray:Array of Texture Object References=@n33, NewItem:"Texture Object Reference (by ref)"=@n32) {is=title K2Node_CallArrayFunction}
   n34 = DuplicateAsset(execute=@n31.else, SourceAssetPath=@n25, DestinationAssetPath=@n28)
   > n35 = Cast To Texture!(execute, Object=@n34) -> (then, CastFailed, AsTexture:Texture Object Reference) {is=title K2Node_DynamicCast}
   > Add(TargetArray=@n33, NewItem=@n35)
   > RenameAsset(Asset=@n34, NewName=@n29)
   PrintString(execute=@n35.CastFailed, InString=Failure Assigning generated texture)
   n36 = Trigger Loading() {is=title K2Node_CustomEvent}
   n37 = Cast To MaterialInstanceConstant(Object=@n7)
   > n38 = SetMaterialInstanceStaticSwitchParameterValue(
    Instance=@n37
    ParameterName=UseCinematicsOutlineThicknessOverrideTexture
    Value=true
    Association=GlobalParameter
    bUpdateMaterialInstance=true
   )
   > SetMaterialInstanceTextureParameterValue(Instance=@n37, ParameterName=CinematicsOutlineThicknessOverride, Association=GlobalParameter)
   PrintString(execute=@n37.CastFailed, InString=ERROR CONVERTING BASE MATERIAL INSTANCE VERSION)
   Event Construct()
   > SetObject(self=Get DetailsView_22() -> (DetailsView_22:Details View Object Reference), NewObject=Self-Reference())
   > n39 = GetSelectionSet(self=Get LevelEditorSubsystem())
   > Bind Event to On Selection Change(self:Typed Element Selection Set Object Reference=@n39, Delegate=Create Event()) {is=title K2Node_AddDelegate}
   Attach Actor To Actor!(
    execute=@n44.then
    self:Actor Object Reference=@n4
    ParentActor:Actor Object Reference=Get Rotator()
    SocketName:Name=None
    LocationRule:EAttachmentRule Enum=SnapToTarget
    RotationRule:EAttachmentRule Enum=KeepRelative
    ScaleRule:EAttachmentRule Enum=KeepRelative
    bWeldSimulatedBodies:bool=TRUE
   ) -> (then, ReturnValue:bool)
   n40 = SpawnActor Skeletal Mesh Actor(
    execute=@n45.then_0
    Class=&SkeletalMeshActor
    SpawnTransform_Location=@n3.OutHit_Location
    SpawnTransform_Rotation="0, 0, 0"
    SpawnTransform_Scale="1,1,1"
    CollisionHandlingOverride=Undefined
    TransformScaleMethod=MultiplyWithRoot
   )
   Attach Actor To Actor(
    execute=@n42.then
    self=@n40
    ParentActor=Get Rotator()
    SocketName=None
    LocationRule=SnapToTarget
    RotationRule=KeepRelative
    ScaleRule=KeepRelative
    bWeldSimulatedBodies=TRUE
   )
   "To String (Boolean)"(InBool=@n38)
   GetAssetRegistry()
   n41 = Get Object Path String(Object=Get LargeEmpty_Base())
   PrintString(InString=@n41)
   String To Name(InString=@n41)
   Append(A=/, B=GetSystemPath(Object=@n24))
   Append!(A=@n26, B=@n29, C:String=., D:String=@n29) -> (ReturnValue)
   Event PreConstruct()
   Event Tick()
   Get EnhancedInputEditorSubsystem() -> (ReturnValue:Enhanced Input Editor Subsystem Object Reference) {is=title K2Node_GetEditorSubsystem}
   Add(
    execute=@n40.then
    TargetArray:Array of Skeletal Mesh Actor Object References=Get Unselectables()
    NewItem:"Skeletal Mesh Actor Object Reference (by ref)"=@n40
   )
   > n42 = SetSkeletalMeshAsset!(execute, self:Skeletal Mesh Component Object Reference=Get SkeletalMeshComponent(self=@n40), NewMesh:Skeletal Mesh Object Reference=@n6) -> (then)
   Add(execute=@n4.then, TargetArray=Get Selectables(), NewItem=@n4)
   > n43 = Sequence!(execute) -> (then_0, then_1)
   > n44 = SetSkeletalMeshAsset(self=@n5, NewMesh=@n6)
   n45 = Sequence(execute=@n6.LoopBody)
   n46 = SpawnActor BP Toonline Adjuster Rotator!(
    execute:Exec=@n3.then
    WorldContextObject:Object Reference
    Class:Actor Class Reference=&/BP_ToonlineAdjuster_Rotator.BP_ToonlineAdjuster_Rotator_C
    SpawnTransform_Location:Vector=@n3.OutHit_Location
    SpawnTransform_Rotation:Rotator="0, 0, 0"
    SpawnTransform_Scale:Vector="1,1,1"
    CollisionHandlingOverride:ESpawnActorCollisionHandlingMethod Enum=Undefined
    TransformScaleMethod:ESpawnActorScaleMethod Enum=MultiplyWithRoot
    Owner:Actor Object Reference
    Instigator:Pawn Object Reference
   ) -> (then:Exec, ReturnValue:BP Toonline Adjuster Rotator Object Reference) {is=title K2Node_SpawnActorFromClass}
   > n47 = Set Rotator(Rotator:BP Toonline Adjuster Rotator Object Reference=@n46) -> (Output_Get:BP Toonline Adjuster Rotator Object Reference)
  ]}
]
variables [
 GeneratedTextures {type="array<Texture>", default []}
 ID {type=AthenaCharacterItemDefinition, default=/BRCosmetics/Athena/Items/Cosmetics/Characters/Character_ClearRadius}
 LargeEmpty_Base {type=Texture, default=/Game/Global/Textures/Default/Blanks/LargeEmpty}
 MIDs {type="array<MaterialInstanceDynamic>", default []}
 OverlayMaterials {type="array<MaterialInterface>", default []}
 OverridesDestination {type=DirectoryPath, default {path=""}}
 ProtectOverlayOutlinerSelection? {type=boolean, default=true}
 Rotator {type=BP_ToonlineAdjuster_Rotator_C, default=null}
 Selectables {type="array<SkeletalMeshActor>", default []}
 SkeletalMeshParts {type="array<SkeletalMesh>", default []}
 Unselectables {type="array<SkeletalMeshActor>", default []}
]
widgets {
 info {rootWidgetClass=&Overlay, widgetCount=9}
 widgets [
  Overlay_0 {widgetClassPath=&Overlay, widget=Overlay_0}
  VerticalBox_0 {
   widgetClassPath=&VerticalBox
   parent=Overlay_0
   widget=VerticalBox_0
   slot {class=OverlaySlot, properties {horizontalAlignment=HAlign_Fill, verticalAlignment=VAlign_Fill}}
  }
  VerticalBox_1 {widgetClassPath=&VerticalBox, parent=VerticalBox_0, widget=VerticalBox_1, slot {class=VerticalBoxSlot}}
  HorizontalBox_0 {widgetClassPath=&HorizontalBox, parent=VerticalBox_1, widget=HorizontalBox_0, slot {class=VerticalBoxSlot}}
  GenerateTextures {
   widgetClassPath=&TA_Generic01_Button
   parent=HorizontalBox_0
   widget=GenerateTextures
   properties {label="Generate Textures for "}
   slot {class=HorizontalBoxSlot}
  }
  LoadCID {
   widgetClassPath=&TA_Generic01_Button
   parent=HorizontalBox_0
   widget=LoadCID
   properties {label=Load CID}
   slot {class=HorizontalBoxSlot, properties {size {sizeRule=Fill, value=1}}}
  }
  Clear {
   widgetClassPath=&TA_Generic01_Button
   parent=HorizontalBox_0
   widget=Clear
   properties {label=Clear, buttonColor_Default=<rgba 1 0.0866 0 1>}
   slot {class=HorizontalBoxSlot}
  }
  DetailsView_22 {
   widgetClassPath=&DetailsView
   parent=VerticalBox_1
   widget=DetailsView_22
   properties {categoriesToShow [NoCategory]}
   slot {class=VerticalBoxSlot}
  }
  TextureGeneration {widgetClassPath=&VerticalBox, parent=VerticalBox_0, widget=TextureGeneration, slot {class=VerticalBoxSlot}}
 ]
}
